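# Niagara water effect for LemonCart fountain A: a top ribbon stream, spout splashes, drips from rings, and pool ripples with splash sprites.
# Spawn groups index User arrays for per-tier offsets and radii, while TopSprayPosition sets the stream and spout height.
# Fortnite time-of-day curves tint the stream, splashes and ripples.
# Drips have GPU and lighter CPU (LOW) emitters, both distance-culled; the Mist emitter is disabled.
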NiagaraSystem /Clyde/Environments/Sets/LemonCart/FX/Fountains/NS_LemonCart_Fountain_A
dependencies {
 usedDataInterfaces [
  &NiagaraDataInterfaceCurve
  &NiagaraDataInterfaceColorCurve
  &NiagaraDataInterfaceVectorCurve
  &NiagaraDataInterfaceSpriteRendererInfo
 ]
 usedDynamicInputs [
  &RandomRangeVector
  &Vector2DFromFloat
  &SimulationPosition
 ]
 usedModules [&SystemState]
}
emitters {
 Drip_Drops {
  simTarget=GPUComputeSim
  emitterUpdateScript [
   &EmitterState(Enable Distance Culling=true, MaxDistance=4800, Module.bUseMaxDistance=true, Module.bUseMinDistance=false, Scalability Mode=::NewEnumerator1)
   &SpawnRate(SpawnRate=50, Use Spawn Probability=false)
   &SpawnRate(SpawnGroup=1, SpawnRate=125, Use Spawn Probability=false) {module=SpawnRate001}
  ]
  particleSpawnScript [
   &PartitionParticles(Number Of Partitions=4)
   &InitializeParticle(
    Lifetime=&UniformRangedFloat(Minimum=0.375, Maximum=0.7)
    Sprite Size=&RandomRangeVector2D(Minimum=<10 30>, Maximum=<30 50>)
    Sprite Size Mode=::NewEnumerator4
    UsePositionOffset=false
    Write Material Random=true
   )
   &SystemLocation(MaskSpawnGroup=false, Offset=<0 0 -35>)
   &ShapeLocation(
    MaskSpawnGroup=false
    Offset=&SelectVectorFromArray(Direct Array Index=&SpawnGroup(), Vector Selection Array=$User.RingOffset)
    Offset Coordinate Space=::Local
    Offset Mode=::NewEnumerator0
    Override Seed=false
    Ring / Disc Distribution Mode:ENiagaraSphereDistributionMode=::NewEnumerator1
    Ring Radius=&Multiply_Float(
     A=&SelectFloatFromArray(Direct Array Index=&SpawnGroup(), Float Selection Array=$User.RingSize)
     B=$User.BaseRadiusScale
    )
    Rotation Angle:NiagaraFloat=&MaskFloatBySpawnGroup(
     Masked Float=&SelectFloatFromArray(
      Direct Array Index=$Particles.PartitionParticles.Partition
      Float Selection Array=DataInterface() {propertyValues {FloatData [0,0.25,0.5,0.75]}}
     )
     Passthrough Float=1
    )
    Rotation Axis=<0 0 1>
    Rotation Mode=::NewEnumerator1
    Shape Primitive=::NewEnumerator4
    U Position=&MaskFloatBySpawnGroup(
     Masked Float=&RandomRangeFloat(Minimum=-0.025, Maximum=0.025)
     Passthrough Float=&RandomRangeFloat(Minimum=0, Maximum=1)
    )
    UseOwnerScale=true
   )
   &CameraOffset(Camera Offset Amount=10)
   &DynamicMaterialParameters(
    Index 0 Param 1=1
    Index 0 Param 2=10
    Index 0 Param 3=100
    Index 0 Param 4=8
    Module.Index0Param0Write=true
    Module.Index0Param1Write=true
    Module.Index0Param2Write=true
    Module.Index0Param3Write=true
   )
  ]
  particleUpdateScript [
   &ParticleState()
   &GravityForce()
   &Drag(Drag=&RandomRangeFloat(Minimum=1, Maximum=3), Use Linear Drag=true, Use Rotational Drag=true)
   &SolveForcesAndVelocity(Clamp Velocity=false, Limit Acceleration=false)
   &ScaleSpriteSizeBySpeed(Max Scale Factor=<1 2>, Min Scale Factor=<0 0.5>)
   &ScaleSpriteSize(Uniform Curve Scale=1.5, Uniform Curve Sprite Scale=DataInterface() {propertyValues {Curve {keys [<at 0 0>,<at 0.1 0.925>,<at 1 0.4>]}}})
   &ScaleColor(
    Scale Alpha=&FloatFromCurve(
     FloatCurve=DataInterface() {propertyValues {Curve {keys [<at 0 -1>, {interpMode=RCIM_Cubic, tangentMode=RCTM_User, time=0.009, value=0.9316}, <at 0.7 0.875>, <at 1 -1>]}}}
     Scale Curve=0.25
    )
    ScaleA=true
    ScaleRGB=true
   )
  ]
  rendererClasses [&NiagaraSpriteRendererProperties]
  renderers [{rendererClass=&NiagaraSpriteRendererProperties}]
 }
 Drip_Drops_LOW {
  emitterUpdateScript [
   &EmitterState(Enable Distance Culling=true, MaxDistance=4800, Module.bUseMaxDistance=true, Module.bUseMinDistance=false, Scalability Mode=::NewEnumerator1)
   &SpawnRate(SpawnRate=32, Use Spawn Probability=false)
   &SpawnRate(SpawnGroup=1, SpawnRate=40, Use Spawn Probability=false) {module=SpawnRate001}
  ]
  particleSpawnScript [
   &PartitionParticles(Number Of Partitions=4)
   &InitializeParticle(
    Lifetime=&UniformRangedFloat(Minimum=0.375, Maximum=0.7)
    Sprite Size=&RandomRangeVector2D(Minimum=<10 30>, Maximum=<30 50>)
    Sprite Size Mode=::NewEnumerator4
    UsePositionOffset=false
    Write Material Random=true
   )
   &SystemLocation(MaskSpawnGroup=false, Offset=<0 0 -35>)
   &ShapeLocation(
    MaskSpawnGroup=false
    Offset=&SelectVectorFromArray(Direct Array Index=&SpawnGroup(), Vector Selection Array=$User.RingOffset)
    Offset Coordinate Space=::Local
    Offset Mode=::NewEnumerator0
    Override Seed=false
    Ring / Disc Distribution Mode:ENiagaraSphereDistributionMode=::NewEnumerator1
    Ring Radius=&Multiply_Float(
     A=&SelectFloatFromArray(Direct Array Index=&SpawnGroup(), Float Selection Array=$User.RingSize)
     B=$User.BaseRadiusScale
    )
    Rotation Angle:NiagaraFloat=&MaskFloatBySpawnGroup(
     Masked Float=&SelectFloatFromArray(
      Direct Array Index=$Particles.PartitionParticles.Partition
      Float Selection Array=DataInterface() {propertyValues {FloatData [0,0.25,0.5,0.75]}}
     )
     Passthrough Float=1
    )
    Rotation Axis=<0 0 1>
    Rotation Mode=::NewEnumerator1
    Shape Primitive=::NewEnumerator4
    U Position=&MaskFloatBySpawnGroup(
     Masked Float=&RandomRangeFloat(Minimum=-0.025, Maximum=0.025)
     Passthrough Float=&RandomRangeFloat(Minimum=0, Maximum=1)
    )
    UseOwnerScale=true
   )
   &CameraOffset(Camera Offset Amount=10)
   &DynamicMaterialParameters(
    Index 0 Param 1=1
    Index 0 Param 2=10
    Index 0 Param 3=100
    Index 0 Param 4=8
    Module.Index0Param0Write=true
    Module.Index0Param1Write=true
    Module.Index0Param2Write=true
    Module.Index0Param3Write=true
   )
  ]
  particleUpdateScript [
   &ParticleState()
   &GravityForce()
   &Drag(Drag=&RandomRangeFloat(Minimum=1, Maximum=3), Use Linear Drag=true, Use Rotational Drag=true)
   &SolveForcesAndVelocity(Clamp Velocity=false, Limit Acceleration=false)
   &ScaleSpriteSizeBySpeed(Max Scale Factor=<1 2>, Min Scale Factor=<0 0.5>)
   &ScaleSpriteSize(Uniform Curve Scale=1.5, Uniform Curve Sprite Scale=DataInterface() {propertyValues {Curve {keys [<at 0 0>,<at 0.1 0.925>,<at 1 0.4>]}}})
   &ScaleColor(
    Scale Alpha=&FloatFromCurve(
     FloatCurve=DataInterface() {propertyValues {Curve {keys [<at 0 -1>, {interpMode=RCIM_Cubic, tangentMode=RCTM_User, time=0.009, value=0.9316}, <at 0.7 0.875>, <at 1 -1>]}}}
     Scale Curve=0.25
    )
    ScaleA=true
    ScaleRGB=true
   )
  ]
  rendererClasses [&NiagaraSpriteRendererProperties]
  renderers [{rendererClass=&NiagaraSpriteRendererProperties}]
 }
 Mist {
  bEnabled=false
  emitterUpdateScript [
   &EmitterState(Enable Distance Culling=true, MaxDistance=4000, Module.bUseMaxDistance=true, Module.bUseMinDistance=false, Scalability Mode=::NewEnumerator1)
   &SpawnRate(SpawnRate=5, Use Spawn Probability=false)
   &SpawnRate(SpawnGroup=1, SpawnRate=8, Use Spawn Probability=false) {module=SpawnRate001}
   &SpawnRate(SpawnGroup=2, SpawnRate=25, Use Spawn Probability=false) {module=SpawnRate002}
  ]
  particleSpawnScript [
   /SetVariables(Particles.Random=&UniformRangedFloat(Minimum=0, Maximum=1)) {isSetParameters=true}
   &InitializeParticle(
    Color=<rgba 0.612 1.0432 1.5 0.8>
    Lifetime Max=1.2
    Lifetime Min=0.8
    Lifetime Mode=::NewEnumerator1
    Sprite Rotation Mode=::NewEnumerator2
    Sprite Size Mode=::NewEnumerator1
    Uniform Sprite Size Max=130
    Uniform Sprite Size Min=95
    UsePositionOffset=false
    Write Material Random=true
   )
   &CameraOffset(Camera Offset Amount=5)
   &SystemLocation(MaskSpawnGroup=false, Offset=<0 0 -36>)
   &ShapeLocation(
    Disc Coverage=0.25
    MaskSpawnGroup=false
    Offset=&SelectVectorFromArray(Direct Array Index=&SpawnGroup(), Vector Selection Array=$User.PoolOffset)
    Offset Coordinate Space=::Local
    Offset Mode=::NewEnumerator0
    Override Seed=false
    Ring Radius=&Multiply_Float(
     A=&SelectFloatFromArray(Direct Array Index=&SpawnGroup(), Float Selection Array=$User.PoolSize)
     B=$User.BaseRadiusScale
    )
    Shape Primitive=::NewEnumerator4
    UseOwnerScale=true
   )
   &DynamicMaterialParameters(
    Index 0 Param 1=&UniformRangedFloat(Minimum=1, Maximum=2)
    Index 0 Param 2=0.2
    Index 0 Param 3=4
    Index 0 Param 4=0.01
    Module.Index0Param0Write=true
    Module.Index0Param1Write=true
    Module.Index0Param2Write=true
    Module.Index0Param3Write=true
   )
  ]
  particleUpdateScript [
   &ParticleState()
   &SubUVAnimation(
    SubUV Animation Mode:ENiagara_SubUVLookupMode=::NewEnumerator1
    SubUV Curve=DataInterface() {propertyValues {Curve {keys [{interpMode=RCIM_Cubic, leaveTangent=2.3356, tangentMode=RCTM_Break, time=0, value=0}, <at 1 0.9>]}}}
   )
   &ScaleSpriteSize(
    Uniform Curve Scale=&MaskFloatBySpawnGroup(Masked Float=0.7, Passthrough Float=1)
    Uniform Curve Sprite Scale=DataInterface() {propertyValues {Curve {keys [{tangentMode=RCTM_User, time=0, value=0.35}, {tangentMode=RCTM_User, time=1, value=1}]}}}
   )
   &ScaleColor(
    Scale Alpha=&FloatFromCurve(FloatCurve=DataInterface() {propertyValues {Curve {keys [
         {arriveTangent=26.1673, interpMode=RCIM_Cubic, leaveTangent=26.1673, tangentMode=RCTM_User, time=0, value=0}
         <at 0.1 1 RCIM_Cubic>
         {arriveTangent=-0.6729, interpMode=RCIM_Cubic, leaveTangent=-0.6729, tangentMode=RCTM_User, time=1, value=0}
        ]}}})
    ScaleA=true
    ScaleRGB=true
   )
  ]
  rendererClasses [&NiagaraSpriteRendererProperties]
  renderers [{rendererClass=&NiagaraSpriteRendererProperties}]
 }
 Pool_Ripples_Splash {
  emitterUpdateScript [
   &EmitterState()
   &SpawnRate(SpawnRate=8, Use Spawn Probability=false)
   &SpawnRate(SpawnGroup=1, SpawnRate=25, Use Spawn Probability=false) {module=SpawnRate001}
   &SpawnRate(SpawnGroup=2, SpawnRate=35, Use Spawn Probability=false) {module=SpawnRate002}
  ]
  particleSpawnScript [
   /SetVariables(Particles.RandomXY:NiagaraFloat=&Multiply_Float(
     A=&RandomRangeFloat(Minimum=0.105, Maximum=0.21)
     B=&SelectFloatFromArray(Direct Array Index=&SpawnGroup(), Float Selection Array=$User.PoolScale)
    )) {isSetParameters=true}
   &InitializeParticle(
    Color=<rgba 1.2 1.2 1.2 0.2>
    Lifetime Max=1.4
    Lifetime Min=0.9
    Lifetime Mode=::NewEnumerator1
    Mesh Scale=&MakeVector(
     X=$Particles.RandomXY
     Y=$Particles.RandomXY
     Z=&Multiply_Float(
      A=&SelectFloatFromArray(Direct Array Index=&SpawnGroup(), Float Selection Array=DataInterface() {propertyValues {FloatData [1,1,2,1]}})
      B=&RandomRangeFloat(Minimum=0.1, Maximum=0.15)
     )
    )
    Mesh Scale Mode=::NewEnumerator4
    Sprite Rotation Mode=::NewEnumerator2
    Sprite Size Mode=::NewEnumerator3
    Sprite UV Mode=::NewEnumerator3
    Uniform Sprite Size=&SelectFloatFromArray(Direct Array Index=&SpawnGroup(), Float Selection Array=DataInterface() {propertyValues {FloatData [8,10,12]}})
    UsePositionOffset=false
    Write Material Random=true
   )
   &SystemLocation(MaskSpawnGroup=false, Offset=<0 0 -36>)
   &ShapeLocation(
    Disc Coverage=0.25
    MaskSpawnGroup=false
    Offset=&SelectVectorFromArray(Direct Array Index=&SpawnGroup(), Vector Selection Array=$User.PoolOffset)
    Offset Coordinate Space=::Local
    Offset Mode=::NewEnumerator0
    Override Seed=false
    Ring Radius=&Multiply_Float(
     A=&SelectFloatFromArray(Direct Array Index=&SpawnGroup(), Float Selection Array=$User.PoolSize)
     B=$User.BaseRadiusScale
    )
    Shape Primitive=::NewEnumerator4
    UseOwnerScale=true
   )
   &CameraOffset(Camera Offset Amount=8)
   &SubUVAnimation(
    "Sprite Renderer "=DataInterface() {propertyValues {SpriteRenderer=&self:Secondary_Splashes_0.NiagaraSpriteRendererProperties_1}}
    SubUV Animation Mode:ENiagara_SubUVLookupModeV2=::NewEnumerator2
    UseEndFrame=false
    UseStartFrame=false
   )
  ]
  particleUpdateScript [
   &ParticleState()
   &ScaleMeshSize(Scale Factor=&VectorFromCurve(VectorCurve=DataInterface() {propertyValues {
       XCurve {keys [
         {interpMode=RCIM_Cubic, leaveTangent=7.6613, tangentMode=RCTM_Break, time=0, value=0}
         {arriveTangent=2.4936, interpMode=RCIM_Cubic, leaveTangent=2.4936, tangentMode=RCTM_User, time=0.2392, value=1.3545}
         <at 1 2>
        ]}
       YCurve {keys [
         {interpMode=RCIM_Cubic, leaveTangent=7.7916, tangentMode=RCTM_Break, time=0, value=0}
         {arriveTangent=2.6017, interpMode=RCIM_Cubic, leaveTangent=2.6017, tangentMode=RCTM_User, time=0.2392, value=1.3545}
         <at 1 2>
        ]}
       ZCurve {keys [<at 0 3>,<at 1 1>]}
      }}))
   &DynamicMaterialParameters(
    Index 0 Param 1=&FloatFromCurve(FloatCurve=DataInterface() {propertyValues {Curve {keys [<at 0 0>,<at 1 1.5>]}}})
    Index 0 Param 2=&FloatFromCurve(
     FloatCurve=DataInterface() {propertyValues {Curve {keys [<at 0 3>,<at 0.1 1>,<at 1 0>]}}}
     Scale Curve=&SelectFloatFromArray(
      Direct Array Index=&SpawnGroup()
      Float Selection Array=DataInterface() {propertyValues {FloatData [0.4,0.1,0.15,0.2]}}
     )
    )
    Index 0 Param 3=10
    Index 0 Param 4=1
    Index 1 Param 1=&FloatFromCurve(FloatCurve=DataInterface() {propertyValues {Curve {keys [
         {arriveTangent=-2.0663, time=0, value=5}
         {arriveTangent=-3.8735, interpMode=RCIM_Cubic, leaveTangent=-3.8735, tangentMode=RCTM_User, time=0.15, value=0.95}
         {arriveTangent=-1.3063, interpMode=RCIM_Cubic, leaveTangent=-1.3063, tangentMode=RCTM_User, time=0.35, value=0}
        ]}}})
    Module.Index0Param0Write=true
    Module.Index0Param1Write=true
    Module.Index0Param2Write=true
    Module.Index0Param3Write=true
    Module.Index1Param0Write=true
    Module.Index1Param1Write=true
    Module.Index1Param2Write=true
    Module.Index1Param3Write=true
    Write Parameter Index 1=true
   )
   &ScaleColor(
    Scale Alpha=&FloatFromCurve(FloatCurve=DataInterface() {propertyValues {Curve {keys [
         <at 0 0>
         {arriveTangent=5.1071, interpMode=RCIM_Cubic, leaveTangent=5.1071, tangentMode=RCTM_Break, time=0.1294, value=2.8567}
         {arriveTangent=-8.2025, interpMode=RCIM_Cubic, leaveTangent=-8.2025, tangentMode=RCTM_User, time=0.3115, value=1.8534}
         <at 0.4715 0.0687>
         <at 1 0>
        ]}}})
    Scale RGB=&Multiply_VectorByFloat(
     Vector=&MakeVectorFromLinearColorRGB(Color=&ColorFromCurve(
       ColorCurve=DataInterface() {propertyValues {
         AlphaCurve {keys []}
         BlueCurve {keys [{arriveTangent=-1.0847, interpMode=RCIM_Cubic, time=0, value=0.9479}, <at 0.25 0.5156>, <at 0.5 0.4752>, <at 0.75 0.3573>, <at 1 0.7708>]}
         GreenCurve {keys [{arriveTangent=-3.7763, interpMode=RCIM_Cubic, time=0, value=0.637}, <at 0.25 0.526>, <at 0.5 0.7188>, <at 0.75 0.3646>, <at 1 0.439>]}
         RedCurve {keys [{arriveTangent=-8.1645, interpMode=RCIM_Cubic, time=0, value=0.1789}, <at 0.25 0.1423>, <at 0.5 0.1326>, <at 0.75 0.0986>, <at 1 0.1393>]}
        }}
       CurveIndex=Linked() {linkedVariable {name=NPC.FortniteNPC.TimeOfDay_0-1, type=NiagaraFloat}}
      ))
     Float=&FloatFromCurve(FloatCurve=DataInterface() {propertyValues {Curve {keys [<at 0 2>,<at 1 0.2>]}}})
    )
    ScaleA=true
    ScaleRGB=true
   )
   /SetVariables(
    Particles.Splash Camera Offset=2
    Particles.Splash Color:LinearColor=&MultiplyLinearColor(
     Linear Color=<rgba 1.25 1.25 1.25 1>
     Scale Factor=&MakeFloat4FromVectorAndFloat(
      "Vector (RGB)"=&Multiply_Vector(
       A=&MakeVectorFromLinearColorRGB(Color=&ColorFromCurve(
         ColorCurve=DataInterface() {propertyValues {
           AlphaCurve {keys []}
           BlueCurve {keys [{arriveTangent=-1.0847, interpMode=RCIM_Cubic, time=0, value=0.7708}, <at 0.25 0.5573>, <at 0.5 1>, <at 0.75 0.5573>, <at 1 0.7708>]}
           GreenCurve {keys [{arriveTangent=-3.7763, interpMode=RCIM_Cubic, time=0, value=0.439}, <at 0.25 0.5037>, <at 0.5 0.8368>, <at 0.75 0.5037>, <at 1 0.439>]}
           RedCurve {keys [{arriveTangent=-8.1645, interpMode=RCIM_Cubic, time=0, value=0.1393}, <at 0.25 0.2728>, <at 0.5 0.4531>, <at 0.75 0.2728>, <at 1 0.1393>]}
          }}
         CurveIndex=Linked() {linkedVariable {name=NPC.FortniteNPC.TimeOfDay_0-1, type=NiagaraFloat}}
        ))
       B=&VectorFromCurve(VectorCurve=DataInterface() {propertyValues {
          XCurve {keys [<at 0 2 RCIM_Cubic>,<at 1 1 RCIM_Cubic>]}
          YCurve {keys [<at 0 2 RCIM_Cubic>,<at 1 1 RCIM_Cubic>]}
          ZCurve {keys [<at 0 2 RCIM_Cubic>,<at 1 1 RCIM_Cubic>]}
         }})
      )
      "Float (Alpha)"=&FloatFromCurve(
       FloatCurve=DataInterface() {propertyValues {Curve {keys [<at 0 0>, {interpMode=RCIM_Cubic, tangentMode=RCTM_User, time=0.002, value=0.9316}, <at 0.2 0.875>, <at 0.35 0>]}}}
       Scale Curve=0.3
      )
     )
    )
    Particles.Splash Position:NiagaraPosition=&AddVectorToPosition(Position=$Particles.Initial.Position, Vector=<0 0 1>)
   ) {isSetParameters=true}
   &ScaleSpriteSize(
    Uniform Curve Scale=1.2
    Uniform Curve Sprite Scale=DataInterface() {propertyValues {Curve {keys [
        {arriveTangent=25.8201, interpMode=RCIM_Cubic, leaveTangent=25.8201, tangentMode=RCTM_User, time=0, value=0.25}
        {arriveTangent=4.5155, interpMode=RCIM_Cubic, leaveTangent=4.5155, tangentMode=RCTM_User, time=0.35, value=5}
       ]}}}
   )
   &SpriteRotationRate(Rotation Rate=&Lerp_Float(A=-75, B=75, Alpha=$Particles.Initial.MaterialRandom))
  ]
  rendererClasses [&NiagaraMeshRendererProperties, &NiagaraSpriteRendererProperties]
  renderers [
   {rendererClass=&NiagaraMeshRendererProperties, rendererIndex=0}
   {rendererClass=&NiagaraSpriteRendererProperties, rendererIndex=1}
  ]
 }
 Top_Spout_Splashes {
  emitterUpdateScript [
   &EmitterState()
   &SpawnRate(SpawnRate=24, Use Spawn Probability=false)
  ]
  particleSpawnScript [
   &InitializeParticle(
    Color=<rgba 0.6979 0.8286 1 1>
    Lifetime Max=0.7
    Lifetime Min=0.5
    Lifetime Mode=::NewEnumerator1
    Sprite Rotation Mode=::NewEnumerator2
    Sprite Size Mode=::NewEnumerator1
    Uniform Sprite Size Max=45
    Uniform Sprite Size Min=25
    UsePositionOffset=true
    Write Material Random=true
   )
   &SubUVAnimation(End Frame=3, SubUV Animation Mode:ENiagara_SubUVLookupMode=::NewEnumerator2)
   &AddVelocity(
    Cone Angle=10
    Cone Axis=<0 0 1>
    Module.UseConeFalloff=false
    Module.UseVelDistribution=false
    Override Seed=false
    Velocity Mode=::NewEnumerator2
    Velocity Speed=&RandomRangeFloat(Minimum=100, Maximum=200)
   )
   &CameraOffset(Camera Offset Amount=20)
   &SystemLocation(
    MaskSpawnGroup=false
    Offset=&MakeVector(X=0, Y=0, Z=&Add_Float(A=35, B=$User.TopSprayPosition))
   )
  ]
  particleUpdateScript [
   &ParticleState()
   &Drag(Use Linear Drag=true, Use Rotational Drag=true)
   &GravityForce(Gravity=<0 0 -800>)
   &SolveForcesAndVelocity(Clamp Velocity=false, Limit Acceleration=false)
   &ScaleSpriteSize(
    Uniform Curve Scale=1.5
    Uniform Curve Sprite Scale=DataInterface() {propertyValues {Curve {keys [{arriveTangent=0.6467, interpMode=RCIM_Cubic, leaveTangent=0.6467, tangentMode=RCTM_User, time=0, value=0.2}, <at 1 1 RCIM_Cubic>]}}}
   )
   &ScaleColor(
    Scale Alpha=&FloatFromCurve(
     FloatCurve=DataInterface() {propertyValues {Curve {keys [{interpMode=RCIM_Cubic, tangentMode=RCTM_User, time=0.2, value=1}, {interpMode=RCIM_Cubic, tangentMode=RCTM_User, time=1, value=0}]}}}
     Scale Curve=&FNTimeOfDayScalar(Night Scale=1, Morning Scale=0.75, Day Scale=0.5, Evening Scale=0.75)
    )
    Scale RGB=&Multiply_Vector(
     A=&MakeVectorFromLinearColorRGB(Color=&ColorFromCurve(
       ColorCurve=DataInterface() {propertyValues {
         AlphaCurve {keys []}
         BlueCurve {keys [{arriveTangent=-1.0847, interpMode=RCIM_Cubic, time=0, value=0.8594}, <at 0.25 1>, <at 0.5 2>, <at 0.75 0.9901>, <at 1 0.8594>]}
         GreenCurve {keys [{arriveTangent=-3.7763, interpMode=RCIM_Cubic, time=0, value=0.3775}, <at 0.25 0.8753>, <at 0.5 1.8605>, <at 0.75 1>, <at 1 0.3775>]}
         RedCurve {keys [{arriveTangent=-8.1645, interpMode=RCIM_Cubic, time=0, value=0.1281}, <at 0.25 0.3385>, <at 0.5 0.5551>, <at 0.75 0.3541>, <at 1 0.1281>]}
        }}
       CurveIndex=Linked() {linkedVariable {name=NPC.FortniteNPC.TimeOfDay_0-1, type=NiagaraFloat}}
      ))
     B=&VectorFromCurve(VectorCurve=DataInterface() {propertyValues {
        XCurve {keys [
          {arriveTangent=-0.3377, interpMode=RCIM_Cubic, time=0, value=2}
          {arriveTangent=-2.0019, interpMode=RCIM_Cubic, leaveTangent=-2.0019, tangentMode=RCTM_User, time=1, value=1}
         ]}
        YCurve {keys [
          {arriveTangent=-0.3377, interpMode=RCIM_Cubic, time=0, value=2}
          {arriveTangent=-2.0019, interpMode=RCIM_Cubic, leaveTangent=-2.0019, tangentMode=RCTM_User, time=1, value=1}
         ]}
        ZCurve {keys [
          {arriveTangent=-0.3377, interpMode=RCIM_Cubic, time=0, value=2}
          {arriveTangent=-2.0019, interpMode=RCIM_Cubic, leaveTangent=-2.0019, tangentMode=RCTM_User, time=1, value=1}
         ]}
       }})
    )
    ScaleA=true
    ScaleRGB=true
   )
   &DynamicMaterialParameters(
    Index 0 Param 1=&FloatFromCurve(FloatCurve=DataInterface() {propertyValues {Curve {keys [
         {arriveTangent=-0.7772, interpMode=RCIM_Cubic, leaveTangent=-0.7772, tangentMode=RCTM_User, time=0, value=1}
         <at 0.7 1>
         {arriveTangent=-2.6733, interpMode=RCIM_Cubic, leaveTangent=-2.6733, tangentMode=RCTM_User, time=1, value=0}
        ]}}})
    Module.Param0WriteEnabled=true
    Module.Param1WriteEnabled=true
    Module.Param2WriteEnabled=true
    Module.Param3WriteEnabled=true
   ) {module=DynamicMaterialParameters001}
  ]
  rendererClasses [&NiagaraSpriteRendererProperties]
  renderers [{rendererClass=&NiagaraSpriteRendererProperties}]
 }
 Top_Stream_Ribbon {
  emitterUpdateScript [
   &EmitterState(Reset Age On Awaken=true, Scalability Mode=::NewEnumerator1)
   &SpawnBurst_Instantaneous(Spawn Count=10, Use Spawn Probability=false, UseLoopCountLimit=true)
  ]
  particleSpawnScript [
   &InitializeParticle(Color=<rgba 1 1 1 0.5>, Lifetime=2, Ribbon Width=10, Ribbon Width Mode=::NewEnumerator1, UsePositionOffset=false, Write Material Random=true)
   &SystemLocation(
    MaskSpawnGroup=false
    Offset=&Lerp_Vector(
     A=<0.5 -0.5 230>
     B=&MakeVector(X=0.5, Y=-0.5, Z=$User.TopSprayPosition)
     Alpha=&ReturnNormalizedExecIndex()
    )
   )
  ]
  particleUpdateScript [
   &ParticleState(
    Kill Particles When Lifetime Has Elapsed=false
    Let Infinitely Lived Particles Die When Emitter Deactivates=true
    Loop Particles Lifetime=true
   )
   &DynamicMaterialParameters(Module.Index0Param0Write=true, Module.Index0Param1Write=true, Module.Index0Param2Write=true, Module.Index0Param3Write=true)
   &ScaleRibbonWidth(Ribbon Width Scale=&FloatFromCurve(
     FloatCurve=DataInterface() {propertyValues {Curve {keys [
         {arriveTangent=3.0919, interpMode=RCIM_Cubic, leaveTangent=3.0919, tangentMode=RCTM_User, time=0, value=0.25}
         {interpMode=RCIM_Cubic, tangentMode=RCTM_User, time=0.5, value=1}
         {arriveTangent=-2.3511, tangentMode=RCTM_User, time=1, value=0.35}
        ]}}}
     CurveIndex=&ReturnNormalizedExecIndex()
    ))
   &ScaleColor(
    Scale Alpha=&FloatFromCurve(FloatCurve=DataInterface() {propertyValues {Curve {keys [
         {arriveTangent=-1.4341, interpMode=RCIM_Cubic, time=0.1, value=1}
         <at 0.3 0.9 RCIM_Cubic>
         {interpMode=RCIM_Cubic, tangentMode=RCTM_Break, time=1, value=1}
        ]}}})
    Scale RGB=&MakeVectorFromLinearColorRGB(Color=&ColorFromCurve(
      ColorCurve=DataInterface() {propertyValues {
        AlphaCurve {keys []}
        BlueCurve {keys [{arriveTangent=-1.0847, interpMode=RCIM_Cubic, time=0, value=0.6719}, <at 0.25 0.4219>, <at 0.5 0.526>, <at 0.75 0.5151>, <at 1 0.6719>]}
        GreenCurve {keys [{arriveTangent=-3.7763, interpMode=RCIM_Cubic, time=0, value=0.3966}, <at 0.25 0.3767>, <at 0.5 0.4605>, <at 0.75 0.5365>, <at 1 0.3966>]}
        RedCurve {keys [{arriveTangent=-8.1645, interpMode=RCIM_Cubic, time=0, value=0.1652}, <at 0.25 0.1824>, <at 0.5 0.1781>, <at 0.75 0.206>, <at 1 0.1652>]}
       }}
      CurveIndex=Linked() {linkedVariable {name=NPC.FortniteNPC.TimeOfDay_0-1, type=NiagaraFloat}}
     ))
    ScaleA=true
    ScaleRGB=true
   )
  ]
  rendererClasses [&NiagaraRibbonRendererProperties]
  renderers [{rendererClass=&NiagaraRibbonRendererProperties}]
 }
}
summary {
 systemName=NS_LemonCart_Fountain_A
 emitters [
  {emitterName=Top_Stream_Ribbon, rendererClasses [&NiagaraRibbonRendererProperties]}
  {emitterName=Top_Spout_Splashes, rendererClasses [&NiagaraSpriteRendererProperties]}
  {
   emitterName=Pool_Ripples_Splash
   rendererClasses [&NiagaraMeshRendererProperties, &NiagaraSpriteRendererProperties]
  }
  {emitterName=Drip_Drops, simTarget=GPUComputeSim, rendererClasses [&NiagaraSpriteRendererProperties]}
  {bEnabled=false, emitterName=Mist, rendererClasses [&NiagaraSpriteRendererProperties]}
  {emitterName=Drip_Drops_LOW, rendererClasses [&NiagaraSpriteRendererProperties]}
 ]
 userVariables [
  User.Alpha {type=NiagaraFloat, defaultValue=0.2, description=""}
  User.BaseRadiusScale {type=NiagaraFloat, defaultValue=1, description=""}
  User.NumberofBeams {type=NiagaraInt32, defaultValue=8}
  User.NumberofPools {type=NiagaraInt32, defaultValue=4}
  User.NumberofRings {type=NiagaraInt32, defaultValue=3}
  User.OverflowScale {
   type=NiagaraDataInterfaceArrayFloat3
   description=""
   defaultValue {
    struct=NiagaraExt_VariableValue_DataInterface
    value {
     dataInterface=&self:NiagaraDataInterfaceArrayFloat3_2
     dataInterfaceClass=&NiagaraDataInterfaceArrayFloat3
    }
   }
  }
  User.PipeSize {type=NiagaraFloat, defaultValue=8, description=""}
  User.PoolOffset {
   type=NiagaraDataInterfaceArrayFloat3
   description=""
   defaultValue {
    struct=NiagaraExt_VariableValue_DataInterface
    value {
     dataInterface=&self:NiagaraDataInterfaceArrayFloat3_3
     dataInterfaceClass=&NiagaraDataInterfaceArrayFloat3
    }
   }
  }
  User.PoolScale {
   type=NiagaraDataInterfaceArrayFloat
   defaultValue {
    struct=NiagaraExt_VariableValue_DataInterface
    value {
     dataInterface=&self:NiagaraDataInterfaceArrayFloat_2
     dataInterfaceClass=&NiagaraDataInterfaceArrayFloat
    }
   }
  }
  User.PoolSize {
   type=NiagaraDataInterfaceArrayFloat
   defaultValue {
    struct=NiagaraExt_VariableValue_DataInterface
    value {
     dataInterface=&self:NiagaraDataInterfaceArrayFloat_1
     dataInterfaceClass=&NiagaraDataInterfaceArrayFloat
    }
   }
  }
  User.RingOffset {
   type=NiagaraDataInterfaceArrayFloat3
   description=""
   defaultValue {
    struct=NiagaraExt_VariableValue_DataInterface
    value {
     dataInterface=&self:NiagaraDataInterfaceArrayFloat3_0
     dataInterfaceClass=&NiagaraDataInterfaceArrayFloat3
    }
   }
  }
  User.RingSize {
   type=NiagaraDataInterfaceArrayFloat
   defaultValue {
    struct=NiagaraExt_VariableValue_DataInterface
    value {
     dataInterface=&self:NiagaraDataInterfaceArrayFloat_0
     dataInterfaceClass=&NiagaraDataInterfaceArrayFloat
    }
   }
  }
  User.SplashOffset {
   type=NiagaraDataInterfaceArrayFloat3
   description=""
   defaultValue {
    struct=NiagaraExt_VariableValue_DataInterface
    value {
     dataInterface=&self:NiagaraDataInterfaceArrayFloat3_1
     dataInterfaceClass=&NiagaraDataInterfaceArrayFloat3
    }
   }
  }
  User.StreamBentPosition {type=NiagaraFloat, defaultValue=115, description=""}
  User.StreamBigPosition {type=NiagaraFloat, defaultValue=108, description=""}
  User.StreamSmallPosition {type=NiagaraFloat, defaultValue=35, description=""}
  User.TopSprayPosition {type=NiagaraFloat, defaultValue=185, description=""}
  User.TopSpraySpawnRate {type=NiagaraFloat, defaultValue=1, description=""}
 ]
}
